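# Physics asset for the Lemon_Lime face accessory, providing limited secondary motion for its hair and right earring.
# Capsule bodies sit on the head and dyn_ bones, with both main hair pieces and the earring constrained to the head.
# Smaller top and bottom hair pieces branch from the two main pieces, except dyn_hairBottom_B, which attaches directly to the head.
# Main pieces and earring have tighter twist limits than the smaller strands.
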PhysicsAsset /BRCosmetics/Accessories/Hats/F_MED_Lemon_Lime_FaceAcc/Meshes/Lemon_Lime_FaceAcc_Physics
bodies {
 dyn_earring_r [{shapeType=Capsule, length=4.5016, radius=1.0106, center=<2.5735 -0.5119 -0.5567>, rotation=<rot 0 90.0004 -89.9997>}]
 dyn_hairBottom_A [{shapeType=Capsule, length=1.5513, radius=1.3107, center=<0.9579 -0.1564 -0.188>, rotation=<rot -4.4858 -97.9957 -66.9589>}]
 dyn_hairBottom_B [{shapeType=Capsule, length=1.96, radius=1.1906, center=<1.058 0.2741 -0.2097>, rotation=<rot 0 -102.4514 -72.4259>}]
 dyn_hairBottom_C [{shapeType=Capsule, length=1.496, radius=1.1589, center=<1.4371 -0.3625 -0.2858>, rotation=<rot 43.6798 -89.9995 -78.6093>}]
 dyn_hairMain [{shapeType=Capsule, length=3.1439, radius=6.7994, center=<2.7933 0 0.0001>, rotation=<rot 90 0 0>}]
 dyn_hairMain_r [{shapeType=Capsule, length=2.6024, radius=4.3638, center=<2.2516 0 0>, rotation=<rot 0 90 90>}]
 dyn_hairTop_A [{shapeType=Capsule, length=1.5658, radius=1.4475, center=<1.3417 0.1369 -0.6063>, rotation=<rot 0 -104.978 -55.975>}]
 dyn_hairTop_B [{shapeType=Capsule, length=2.4824, radius=2.0602, center=<1.273 -0.0167 0.5347>, rotation=<rot -24.1309 -80.254 -72.1832>}]
 dyn_hairTop_C [{shapeType=Capsule, length=2.8611, radius=1.8507, center=<2.2894 0.1255 0.1794>, rotation=<rot 0 -75.9318 -97.6075>}]
 dyn_hairTop_D [{shapeType=Capsule, length=1.99, radius=1.4044, center=<1.5101 0.2412 0.0127>, rotation=<rot 25.2739 -89.9997 89.1914>}]
 dyn_hairTop_E [{shapeType=Capsule, length=1.7193, radius=1.3307, center=<0.8204 0.2461 0.5776>, rotation=<rot -41.1845 -90 -29.023>}]
 head [{shapeType=Capsule, length=1, radius=2}]
}
constraints [
 {bone1Name=dyn_hairMain, bone2Name=head, swing1LimitDegrees=16, swing2LimitDegrees=4, twistLimitDegrees=1}
 {bone1Name=dyn_hairTop_B, bone2Name=dyn_hairMain, swing1LimitDegrees=25, swing2LimitDegrees=25, twistLimitDegrees=16}
 {bone1Name=dyn_hairTop_C, bone2Name=dyn_hairMain, swing1LimitDegrees=25, swing2LimitDegrees=25, twistLimitDegrees=16}
 {bone1Name=dyn_hairTop_D, bone2Name=dyn_hairMain, swing1LimitDegrees=25, swing2LimitDegrees=25, twistLimitDegrees=16}
 {bone1Name=dyn_hairTop_E, bone2Name=dyn_hairMain, swing1LimitDegrees=25, swing2LimitDegrees=25, twistLimitDegrees=16}
 {bone1Name=dyn_hairBottom_C, bone2Name=dyn_hairMain, swing1LimitDegrees=25, swing2LimitDegrees=25, twistLimitDegrees=16}
 {bone1Name=dyn_hairBottom_A, bone2Name=dyn_hairMain_r, swing1LimitDegrees=30, swing2LimitDegrees=30, twistLimitDegrees=30}
 {bone1Name=dyn_hairBottom_B, bone2Name=head, swing1LimitDegrees=30, swing2LimitDegrees=30, twistLimitDegrees=30}
 {bone1Name=dyn_hairTop_A, bone2Name=dyn_hairMain_r, swing1LimitDegrees=30, swing2LimitDegrees=30, twistLimitDegrees=30}
 {bone1Name=dyn_hairMain_r, bone2Name=head, swing1LimitDegrees=20, swing2LimitDegrees=20, twistLimitDegrees=1}
 {bone1Name=dyn_earring_r, bone2Name=head, swing1LimitDegrees=12, swing2LimitDegrees=8, twistLimitDegrees=4}
]
