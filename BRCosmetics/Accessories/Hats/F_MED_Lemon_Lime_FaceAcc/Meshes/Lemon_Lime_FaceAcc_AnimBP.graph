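# Animation Blueprint for the Lemon Lime face accessory, adding secondary motion on top of the pose copied from its attached parent mesh.
# RigidBody simulates in spine_05 space, then two additive bone offsets nudge the right earring before AnimDynamics swings it.
# The offsets are weighted by head yaw relative to spine_05 and by targeting state.
# Head twist and swing target values are also computed but unused in the AnimGraph.
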AnimBlueprint /BRCosmetics/Accessories/Hats/F_MED_Lemon_Lime_FaceAcc/Meshes/Lemon_Lime_FaceAcc_AnimBP
target_skeleton=&/Lemon_Lime_FaceAcc_Skeleton
graphs [
 AnimGraph {nodes [
   n1 = RigidBody(ComponentPose=Local To Component(LocalPose=Copy Pose From Mesh() {bUseAttachedParent=true})) {
    bClampLinearTranslationLimitToRefPose=true
    bOverrideWorldGravity=true
    overridePhysicsAsset=&/Lemon_Lime_FaceAcc_Physics
    simulationSpace=BaseBoneSpace
    alphaBoolBlend {blendOption=Linear}
    baseBoneRef {boneName=spine_05}
    componentAppliedLinearAccClamp=<2000 2000 2000>
    componentLinearAccScale=<0.4 0.4 0.4>
    componentLinearVelScale=<0.25 0.25 0.25>
    overrideWorldGravity {z=-500}
   }
   n2 = "Transform (Modify) Bone - Bone: dyn_earring_r"(
    ComponentPose=@n1
    Alpha=Get headSwingDelta() -> (headSwingDelta:double)
    Translation="-0.05,-0.000,0.25"
    Rotation="14,0.000,0.000"
    Scale="1,1,1"
   ) {
    rotationMode=BMM_Additive
    rotationSpace=BCS_BoneSpace
    translationMode=BMM_Additive
    translationSpace=BCS_BoneSpace
    alphaBoolBlend {blendOption=Linear}
    boneToModify {boneName=dyn_earring_r}
    rotation {pitch=14}
    translation {x=-0.05, z=0.25}
   }
   n3 = "Transform (Modify) Bone - Bone: dyn_earring_r"(
    ComponentPose=@n2
    Alpha=Get targetAlpha() -> (targetAlpha:double)
    Translation="-0.05,-0.000,0.25"
    Rotation="14,0.000,0.000"
    Scale="1,1,1"
   ) {
    rotationMode=BMM_Additive
    rotationSpace=BCS_BoneSpace
    translationMode=BMM_Additive
    translationSpace=BCS_BoneSpace
    alphaBoolBlend {blendOption=Linear}
    boneToModify {boneName=dyn_earring_r}
    rotation {pitch=14}
    translation {x=-0.05, z=0.25}
   }
   n4 = Anim Dynamics - Bone: dyn_earring_r(ComponentPose=@n3, Alpha=1, ExternalForce="0,0,0") {
    angularDampingOverride=0.999
    bGravityOverrideInSimSpace=true
    bOverrideAngularDamping=true
    bUseGravityOverride=true
    alphaBoolBlend {blendOption=Linear}
    boundBone {boneName=dyn_earring_r}
    componentLinearAccScale=<0.5 0.5 0.5>
    componentLinearVelScale=<0.5 0.5 0.5>
    gravityOverride {z=-980}
    physicsBodyDefinitions [{
      sphereCollisionRadius=0
      boundBone {boneName=dyn_earring_r}
      boxExtents=<2 10 10>
      constraintSetup {angularLimitsMax=<7.5 7.5 7.5>, angularLimitsMin=<-7.5 -7.5 -7.5>, angularTarget=<0 0 0>}
      localJointOffset=<5 0 0>
     }]
   }
   Output Pose(Result=Component To Local(ComponentPose=@n4)) {layerGroup=None}
  ]}
 EventGraph {nodes [
   Event BlueprintUpdateAnimation()
   n1 = BreakTransform(InTransform=GetDeltaTransformFromRefPose(self=GetOwnerSkeletalMeshComponent(), BoneName=head, BaseName=spine_05))
   > n2 = Set headSwingDelta(headSwingDelta:double=MapRangeClamped(Value=BreakRotator(InRot=@n1.Rotation).Yaw, InRangeA=17, InRangeB=-44)) -> (Output_Get:double)
   n3 = BreakRotator(InRot=@n1.Rotation)
   > Set headTwist_r(headTwist_r:double=MapRangeClamped(Value=@n3, InRangeA=15, InRangeB=-37)) -> (Output_Get:double)
   > Set headTwist_l(headTwist_l:double=MapRangeClamped(Value=@n3, InRangeA=15, InRangeB=65)) -> (Output_Get:double)
   > Set headSwingDeltaTarget_r(headSwingDeltaTarget_r:double=float * float(A=0, B=@n2)) -> (Output_Get:double)
   > Set targetAlpha(targetAlpha:double="To Float (Boolean)"(InBool=Get bIsTargeting())) -> (Output_Get:double)
   TryGetPawnOwner()
  ]}
]
variables [
 headSwingDelta {type=number}
 headSwingDeltaTarget_r {type=number}
 headTwist_l {type=number}
 headTwist_r {type=number}
 targetAlpha {type=number}
]
